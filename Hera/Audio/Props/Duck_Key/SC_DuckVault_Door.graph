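# Sound cue for a Duck Vault door, built from a reused elevator door-down recording.
# A Modulator feeds the single wave player at a fixed raised pitch, and the cue boosts volume within an ambient props sound class.
# Overridden attenuation uses logarithmic falloff with occlusion low-pass filtering.
# Overridden concurrency allows one instance per owner and blocks new plays while it sounds.
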SoundCue /Hera/Audio/Props/Duck_Key/SC_DuckVault_Door
root=Modulator_0
nodes [
 {id=Modulator_0, children [WavePlayer_1], properties {pitchMax=1.1, pitchMin=1.1}}
 {id=WavePlayer_1, properties {soundWaveAssetPtr=&/Elevator_Door_Down}}
]
properties {
 bOverrideAttenuation=true
 bOverrideConcurrency=true
 maxDistance=4100
 soundClassObject=&Fort_Ambient_Props_Athena
 volumeMultiplier=1.3
 attenuationOverrides {bEnableOcclusion=true, distanceAlgorithm=Logarithmic, occlusionLowPassFilterFrequency=1000, attenuationShapeExtents {x=500}}
 concurrencyOverrides {bLimitToOwner=true, maxCount=1, resolutionRule=PreventNew}
}
