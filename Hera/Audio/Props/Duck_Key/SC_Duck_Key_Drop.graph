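# Sound cue for the drop sound of the Duck Key prop.
# A single wave player feeds the Duck_Key_Drop sound wave straight to the output, with no randomization, modulation or mixing nodes.
# The cue slightly lowers the volume through its multiplier and sets no attenuation of its own.
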SoundCue /Hera/Audio/Props/Duck_Key/SC_Duck_Key_Drop
root=WavePlayer_1
nodes [{id=WavePlayer_1, properties {soundWaveAssetPtr=&/Duck_Key_Drop}}]
properties {maxDistance=2097152, volumeMultiplier=0.8}
