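# Sound cue for grabbing the key on the Duck Key prop, as its name indicates.
# A single wave player node serves as the root and plays the Duck_Key_Grab_Key sound wave directly, with no randomization, mixing or modulation nodes.
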SoundCue /Hera/Audio/Props/Duck_Key/SC_Duck_Key_Grab_Key_Cue
root=WavePlayer_0
nodes [{id=WavePlayer_0, properties {soundWaveAssetPtr=&/Duck_Key_Grab_Key}}]
properties {maxDistance=2097152}
