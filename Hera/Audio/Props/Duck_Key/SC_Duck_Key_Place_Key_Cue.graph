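# Sound cue for placing a duck key prop, judging by its name and asset path.
# A single wave player feeds the Duck_Key_Place_Key sound wave straight to the output, with no randomization or modulation nodes.
# The cue uses the ATT_DuckKey attenuation asset for spatial falloff and is assigned to the Fort_Gadgets sound class.
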SoundCue /Hera/Audio/Props/Duck_Key/SC_Duck_Key_Place_Key_Cue
root=WavePlayer_0
nodes [{id=WavePlayer_0, properties {soundWaveAssetPtr=&/Duck_Key_Place_Key}}]
properties {
 attenuationSettings=&/ATT_DuckKey
 maxDistance=2474.917
 soundClassObject=&Fort_Gadgets
}
