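# Sound cue that plays the Duck Key reveal sound, used as a prop sound under the gadgets sound class.
# A single wave player feeds the root directly, with spatialization and falloff from the Duck Key attenuation asset.
# A concurrency override limits it to one voice at a time, with timed voice-steal release and volume-scale attack.
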SoundCue /Hera/Audio/Props/Duck_Key/SC_Duck_Key_Reveal_Cue
root=WavePlayer_0
nodes [{id=WavePlayer_0, properties {soundWaveAssetPtr=&/Duck_Key_Reveal}}]
properties {
 attenuationSettings=&/ATT_DuckKey
 bOverrideConcurrency=true
 maxDistance=2474.917
 soundClassObject=&Fort_Gadgets
 volumeMultiplier=1
 concurrencyOverrides {maxCount=1, voiceStealReleaseTime=3, volumeScaleAttackTime=2}
}
